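# Masked, two-sided lit material for Dinosaur Park brontosaurus destruction debris on mesh particles.
# Base color combines the shell texture, darkened backfaces, particle color and particle-relative world-aligned cracks, simplified on mobile and lower quality.
# A smoke texture drives erosion dissolve through particle alpha, with particle red scaling its UVs.
# Switchable emissive uses particle color, dynamic parameters or fresnel, while a RankedAurum blend remaps specular, metallic and roughness channels and adds fresnel glow.
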Material /Game/Effects/Fort_Effects/World/Destruction/Dino/Materials/M_Destruction_Brontosaurus
graph {
 properties {
  bUsedWithMeshParticles=true
  bUsedWithNiagaraMeshParticles=true
  blendMode=BLEND_Masked
  pixelDepthOffsetMode=PDOM_Legacy
  twoSided=true
  usedShadingModels=Default Lit
 }
 nodes [
  ComponentMask() {g=true, r=true}
  n1 = TextureSampleParameter2D() {
   parameterName=Specular
   samplerType=SAMPLERTYPE_Masks
   sortPriority=1
   texture=/Game/Environments/Props/DinosaurPark/Textures/T_Dinosaur_Park_Brontosaurus_Shell_S
  }
  n2 = TextureSampleParameter2D() {
   parameterName=Specular
   samplerType=SAMPLERTYPE_Masks
   sortPriority=1
   texture=&T_Dinosaur_Park_Brontosaurus_Shell_S
  }
  n3 = ChannelMaskParameter(None=@n2.RGB) {group=RankedAurum, maskChannel=Blue, parameterName=RankedAurum_MetalChannel, defaultValue=<rgba 0 0 1 0>}
  LinearInterpolate(
   A=Multiply(A=ScalarParameter() {defaultValue=1, parameterName=Metallic}, B=@n1.G)
   B=Multiply(A=@n3.R, B=ScalarParameter() {defaultValue=1, parameterName=Metallic})
   Alpha=ScalarParameter() {group=RankedAurum, parameterName=UseRankedAurum, sortPriority=0}
  )
  n4 = ChannelMaskParameter(None=@n2.RGB) {group=RankedAurum, maskChannel=Green, parameterName=RankedAurum_RoughnessChannel, defaultValue=<rgba 0 1 0 0>}
  LinearInterpolate(
   A=Multiply(A=@n1.B, B=ScalarParameter() {defaultValue=1, parameterName=Roughness})
   B=Multiply(A=@n4.R, B=ScalarParameter() {defaultValue=1, parameterName=Roughness})
   Alpha=ScalarParameter() {group=RankedAurum, parameterName=UseRankedAurum, sortPriority=0}
  )
  LinearInterpolate(
   A=@n1.R
   B=ChannelMaskParameter(None=@n2.RGB) {group=RankedAurum, parameterName=RankedAurum_SpecularChannel}.R
   Alpha=ScalarParameter() {group=RankedAurum, parameterName=UseRankedAurum, sortPriority=0}
  )
  n5 = TextureSampleParameter2D() {group=Emissive, parameterName=EmissiveMask, texture=/Game/Environments/Props/DinosaurPark/Textures/T_Dinosaur_Park_Brontosaurus_Shell_D}
  n6 = ParticleColor()
  n7 = Multiply(A=@n5.RGB, B=@n6.RGB)
  n8 = Constant() {r=1}
  n9 = Constant()
  n10 = MaterialFunctionCall(Color_Morning=@n8, Color_Day=@n8, Color_Evening=@n9, Color_Night=@n9) {materialFunction=&Color_Time_of_Day}
  n11 = DynamicParameter() {defaultValue=<rgba 0 1 1 1>, paramNames [EmissiveIntensity, Param2, Param3, Param4]}
  n12 = ScalarParameter() {group=Emissive, parameterName=EmissiveIntensity}
  n13 = StaticSwitchParameter(
   True=DynamicParameter() {defaultValue=<rgba 0 1 1 1>, paramNames [EmissiveIntensity, Param2, Param3, Param4]}.EmissiveIntensity
   False=ScalarParameter() {group=Emissive, parameterName=EmissiveIntensity}
  ) {group=Emissive, parameterName=UseDP_Emissive}
  n14 = StaticSwitchParameter(True=Multiply(A=@n6.A, B=@n13), False=@n13) {group=Emissive, parameterName=UseParticleColorToControlEmissiveIntensity}
  n15 = LinearInterpolate(
   A=VectorParameter() {group=Emissive, parameterName=LowColor}.RGB
   B=VectorParameter() {group=Emissive, parameterName=HighColor}.RGB
   Alpha=Fresnel(
    ExponentIn=ScalarParameter() {defaultValue=5, group=Emissive, parameterName=ExponentIn}
    BaseReflectFractionIn=ScalarParameter() {group=Emissive, parameterName=BaseReflectFractionIn}
   )
  )
  n16 = StaticSwitchParameter(
   True=LinearInterpolate(
    B=Multiply(A=@n7, B=@n10.Result)
    Alpha=StaticSwitchParameter(True=@n11.EmissiveIntensity, False=@n12) {group=Emissive, parameterName=UseDP_Emissive}
   )
   False=StaticSwitchParameter(True=Multiply(A=@n14, B=@n15), False=@n13) {group=Emissive, parameterName=UseFresnelEmissive}
  ) {group=Emissive, parameterName=UseParticleColorControlEmissive}
  n17 = Multiply(
   A=VectorParameter() {group=RankedAurum, parameterName=RankedAurum_FresnelColor, defaultValue=<rgba 0 0 0 1>}.RGB
   B=MaterialFunctionCall(
    Brightness=ScalarParameter() {defaultValue=1, group=RankedAurum, parameterName=RankedAurum_FresnelBrightness, sortPriority=33}
    Exponent=ScalarParameter() {defaultValue=1, group=RankedAurum, parameterName=RankedAurum_FresnelExponent, sortPriority=33}
   ) {materialFunction=&MF_Fresnel_FX}.Fresnel Inverted
  )
  MaterialFunctionCall(Emissive=LinearInterpolate(A=@n16, B=Add(A=@n16, B=@n17), Alpha=ScalarParameter() {group=RankedAurum, parameterName=UseRankedAurum, sortPriority=0})) {materialFunction=&MF_FX_EmissiveAdjustments}
  n18 = ParticleColor()
  n19 = Multiply(A=TextureCoordinate(), B=@n18.R)
  StaticSwitchParameter(
   True=MaterialFunctionCall!(
    Ramp=TextureSampleParameter2D(UVs=@n19) {parameterName=ErodeTexture, sortPriority=3, texture=&T_Inky_Smoke_Tile}.R
    Dissolve=OneMinus(None=@n18.A)
    EdgeSharpness
   ) {materialFunction=/Game/Effects/Fort_Effects/Materials/_Functions/MF_Dissolve}.ResultWithOutClamp
   False=Constant() {r=1}
  ) {defaultValue=true, parameterName=UseErode}
  n20 = ParticleColor()
  n21 = Constant() {r=1}
  n22 = TextureSampleParameter2D() {parameterName=BaseColor, texture=&T_Dinosaur_Park_Brontosaurus_Shell_D}
  n23 = StaticSwitchParameter(True=LinearInterpolate(A=Multiply(A=@n22.RGB) {constB=0.55}, B=@n22.RGB, Alpha=TwoSidedSign()), False=@n22.RGB) {defaultValue=true, parameterName=UseTowSidedSign}
  n24 = ScalarParameter() {defaultValue=1, parameterName=CrackAmount}
  n25 = MaterialFunctionCall(
   TextureObject=TextureObject() {texture=/Game/Global/Textures/Default/Masks_and_Gradients/Cracks_02}
   TextureSize=Constant() {r=350}
   WorldPosition=Subtract(A=WorldPosition() {worldPositionShaderOffset=WPT_ExcludeAllShaderOffsets}.XYZ, B=ParticlePositionWS())
  ) {materialFunction=&WorldAlignedTexture}
  n26 = MaterialFunctionCall(Transition Phase=@n24, Height Texture=ComponentMask(None=@n25.XYZ Texture) {b=true}, Contrast=Constant() {r=2}) {materialFunction=&HeightLerp}
  n27 = ComponentMask(None=@n25.XYZ Texture) {r=true}
  n28 = Add(A=OneMinus(None=Multiply(A=@n26.Alpha, B=@n27)), B=Multiply(A=@n26.Alpha, B=ComponentMask(None=@n25.XYZ Texture) {g=true})) {desc=69 - 26 with uv coordinates}
  n29 = LinearInterpolate(A=Constant() {r=1}, B=OneMinus(None=@n27), Alpha=@n24)
  n30 = StaticSwitchParameter(
   True=QualitySwitch(Default=ShadingPathSwitch(Default=@n28, Mobile=@n29) {desc=Migrated from FeatureLevelSwitch}, Low=@n29, Medium=@n29)
   False=Constant() {r=1}
  ) {defaultValue=true, parameterName=UseCrack}
  StaticSwitchParameter(
   True=Multiply(
    A=StaticSwitchParameter(True=@n20.RGB, False=@n21) {defaultValue=true, parameterName=UseParticleColor_Multiply_w_Diff}
    B=Multiply(A=@n23, B=@n30)
   )
   False=ParticleColor().RGB
  ) {defaultValue=true, parameterName=UseParticleColorWithDiffuseTexture}
  TextureSampleParameter2D() {
   parameterName=NormalMap
   samplerType=SAMPLERTYPE_Normal
   sortPriority=2
   texture=/Game/Environments/Props/DinosaurPark/Textures/T_Dinosaur_Park_Brontosaurus_Shell_N
  }
 ]
}
